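# Level Sequence for a medium shot in the PilgrimCore camera manager's set of sequences, lasting 600 frames at 60 fps.
# A spawnable Cine Camera Actor holds a fixed position and rotation through single constant transform keys, with a camera shake section across the shot.
# Its camera component tracks set constant focal length, manual focus distance and a wide aperture.
# One camera cut section spans the whole range.
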LevelSequence /PilgrimCore/CameraManager/Sequences/MediumShots/LS_PilgrimCam_MS_01
bindings [
 CameraComponent {
  id=#1
  parent_id=#2
  possessed_class=CineCameraComponent
  tracks [
   {
    type=MovieSceneFloatTrack
    display_name=Current Focal Length
    sections [{type=MovieSceneFloatSection, channels [None {type=MovieSceneScriptingFloatChannel, default=39.109}]}]
   }
   {
    type=MovieSceneFloatTrack
    display_name="Manual Focus Distance (Focus Settings)"
    sections [{type=MovieSceneFloatSection, channels [None {type=MovieSceneScriptingFloatChannel, default=1279.1984}]}]
   }
   {
    type=MovieSceneFloatTrack
    display_name=Current Aperture
    sections [{type=MovieSceneFloatSection, channels [None {type=MovieSceneScriptingFloatChannel, default=1}]}]
   }
  ]
 }
 Cine Camera Actor {
  id=#2
  possessed_class=CineCameraActor
  tracks [
   {
    type=MovieScene3DTransformTrack
    sections [{
      type=MovieScene3DTransformSection
      channels [
       Location.X {type=MovieSceneScriptingDoubleChannel, default=-3196.7014, const [-4134.9834,1]}
       Location.Y {type=MovieSceneScriptingDoubleChannel, default=196.3234, const [1481.9504,1]}
       Location.Z {type=MovieSceneScriptingDoubleChannel, default=680.9383, const [646.8133,1]}
       Rotation.X {type=MovieSceneScriptingDoubleChannel, default=0, const [-0.1483,1]}
       Rotation.Y {type=MovieSceneScriptingDoubleChannel, default=-4.0714, const [4.3442,1]}
       Rotation.Z {type=MovieSceneScriptingDoubleChannel, default=-169.496, const [-137.9934,1]}
       Scale.X {type=MovieSceneScriptingDoubleChannel, default=1}
       Scale.Y {type=MovieSceneScriptingDoubleChannel, default=1}
       Scale.Z {type=MovieSceneScriptingDoubleChannel, default=1}
      ]
     }]
   }
   {type=MovieSceneSpawnTrack, sections [{type=MovieSceneSpawnSection, channels [None {type=MovieSceneScriptingBoolChannel, default=true}]}]}
   {type=MovieSceneCameraShakeTrack, sections [{type=MovieSceneCameraShakeSection, end_frame=600}]}
  ]
 }
]
display_rate {denominator=1, numerator=60}
tick_resolution {denominator=1, numerator=24000}
tracks [{type=MovieSceneCameraCutTrack, sections [{type=MovieSceneCameraCutSection, end_frame=600, start_frame=0}]}]
