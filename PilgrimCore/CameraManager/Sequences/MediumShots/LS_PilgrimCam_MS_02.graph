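# LS_PilgrimCam_MS_02 is a medium-shot camera Level Sequence for the PilgrimCore camera manager, running 600 frames at 60 fps.
# A spawned CineCameraActor drives the camera cut, drifting slowly with a gradual pan through transform keys.
# Its CineCameraComponent tracks hold focal length, manual focus distance and aperture constant.
# Y location and rotation keys change abruptly between frame 599 and the end key at 600.
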LevelSequence /PilgrimCore/CameraManager/Sequences/MediumShots/LS_PilgrimCam_MS_02
bindings [
 CameraComponent {
  id=#1
  parent_id=#2
  possessed_class=CineCameraComponent
  tracks [
   {
    type=MovieSceneFloatTrack
    display_name=Current Focal Length
    sections [{type=MovieSceneFloatSection, channels [None {type=MovieSceneScriptingFloatChannel, default=13.2198}]}]
   }
   {
    type=MovieSceneFloatTrack
    display_name="Manual Focus Distance (Focus Settings)"
    sections [{type=MovieSceneFloatSection, channels [None {type=MovieSceneScriptingFloatChannel, default=854.2014}]}]
   }
   {
    type=MovieSceneFloatTrack
    display_name=Current Aperture
    sections [{type=MovieSceneFloatSection, channels [None {type=MovieSceneScriptingFloatChannel, default=1}]}]
   }
  ]
 }
 Cine Camera Actor {
  id=#2
  possessed_class=CineCameraActor
  tracks [
   {
    type=MovieScene3DTransformTrack
    sections [{
      type=MovieScene3DTransformSection
      channels [
       Location.X {type=MovieSceneScriptingDoubleChannel, default=-1415.9852, thinned_from=3, keys [[0,-5944.0068],[600,-5900.2476]]}
       Location.Y {type=MovieSceneScriptingDoubleChannel, default=-1657.2339, keys [[0,-1027.9489],[599,-724.0938],[600,-967.5891]]}
       Location.Z {type=MovieSceneScriptingDoubleChannel, default=1225.1136, thinned_from=3, keys [[0,747.8093],[600,755.5702]]}
       Rotation.X {type=MovieSceneScriptingDoubleChannel, default=0, keys [[0,0.0449],[599,-0.1045],[600,-5.3371]]}
       Rotation.Y {type=MovieSceneScriptingDoubleChannel, default=-14.0857, keys [[0,6.2404],[599,7.6962],[600,3.7172]]}
       Rotation.Z {type=MovieSceneScriptingDoubleChannel, default=176.5858, keys [[0,133.155],[599,145.262],[600,139.9556]]}
       Scale.X {type=MovieSceneScriptingDoubleChannel, default=1, const [1,2]}
       Scale.Y {type=MovieSceneScriptingDoubleChannel, default=1, const [1,2]}
       Scale.Z {type=MovieSceneScriptingDoubleChannel, default=1, const [1,2]}
      ]
     }]
   }
   {type=MovieSceneSpawnTrack, sections [{type=MovieSceneSpawnSection, channels [None {type=MovieSceneScriptingBoolChannel, default=true}]}]}
  ]
 }
]
display_rate {denominator=1, numerator=60}
tick_resolution {denominator=1, numerator=24000}
tracks [{type=MovieSceneCameraCutTrack, sections [{type=MovieSceneCameraCutSection, end_frame=600, start_frame=0}]}]
