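# Level Sequence for medium shot 03 in the PilgrimCore CameraManager set, framing a fixed view through one Cine Camera Actor.
# The actor has a spawn track and a constant transform, while its CineCameraComponent holds focal length, manual focus distance and aperture without animation.
# A camera shake track on the component is the only possible source of motion, alongside a single camera cut spanning frames 0 to 600.
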LevelSequence /PilgrimCore/CameraManager/Sequences/MediumShots/LS_PilgrimCam_MS_03
bindings [
 CameraComponent {
  id=#1
  parent_id=#2
  possessed_class=CineCameraComponent
  tracks [
   {
    type=MovieSceneFloatTrack
    display_name=Current Focal Length
    sections [{type=MovieSceneFloatSection, channels [None {type=MovieSceneScriptingFloatChannel, default=12.6318}]}]
   }
   {
    type=MovieSceneFloatTrack
    display_name="Manual Focus Distance (Focus Settings)"
    sections [{type=MovieSceneFloatSection, channels [None {type=MovieSceneScriptingFloatChannel, default=897.6984, const [897.6984,1]}]}]
   }
   {
    type=MovieSceneFloatTrack
    display_name=Current Aperture
    sections [{type=MovieSceneFloatSection, channels [None {type=MovieSceneScriptingFloatChannel, default=1}]}]
   }
   {type=MovieSceneCameraShakeTrack, sections [{type=MovieSceneCameraShakeSection, end_frame=640}]}
  ]
 }
 Cine Camera Actor {
  id=#2
  possessed_class=CineCameraActor
  tracks [
   {
    type=MovieScene3DTransformTrack
    sections [{
      type=MovieScene3DTransformSection
      channels [
       Location.X {type=MovieSceneScriptingDoubleChannel, default=-4621.3414, const [-7252.7095,1]}
       Location.Y {type=MovieSceneScriptingDoubleChannel, default=-1954.016, const [-1342.9692,1]}
       Location.Z {type=MovieSceneScriptingDoubleChannel, default=702.2314, const [642.898,1]}
       Rotation.X {type=MovieSceneScriptingDoubleChannel, default=0, const [-0.5172,1]}
       Rotation.Y {type=MovieSceneScriptingDoubleChannel, default=-5.4146, const [16.2967,1]}
       Rotation.Z {type=MovieSceneScriptingDoubleChannel, default=159.4499, const [44.7307,1]}
       Scale.X {type=MovieSceneScriptingDoubleChannel, default=1, const [1,1]}
       Scale.Y {type=MovieSceneScriptingDoubleChannel, default=1, const [1,1]}
       Scale.Z {type=MovieSceneScriptingDoubleChannel, default=1, const [1,1]}
      ]
     }]
   }
   {type=MovieSceneSpawnTrack, sections [{type=MovieSceneSpawnSection, channels [None {type=MovieSceneScriptingBoolChannel, default=true}]}]}
  ]
 }
]
display_rate {denominator=1, numerator=60}
tick_resolution {denominator=1, numerator=24000}
tracks [{type=MovieSceneCameraCutTrack, sections [{type=MovieSceneCameraCutSection, end_frame=600, start_frame=0}]}]
