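# Level Sequence for a medium portrait camera shot in the PilgrimCore camera manager, running ten seconds at 60 fps on a single cine camera.
# Transform keys drift and rotate the camera, including a marked roll, with a small position jump on the final frame.
# Manual focus pulls from far to near while focal length and a very wide aperture stay fixed for shallow depth of field.
# A camera cut track spans the whole shot.
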LevelSequence /PilgrimCore/CameraManager/Sequences/MediumShots/LS_PilgrimCam_MS_04_Portrait
bindings [
 CameraComponent {
  id=#1
  parent_id=#2
  possessed_class=CineCameraComponent
  tracks [
   {
    type=MovieSceneFloatTrack
    display_name=Current Focal Length
    sections [{type=MovieSceneFloatSection, channels [None {type=MovieSceneScriptingFloatChannel, default=46}]}]
   }
   {
    type=MovieSceneFloatTrack
    display_name="Manual Focus Distance (Focus Settings)"
    sections [{type=MovieSceneFloatSection, channels [None {type=MovieSceneScriptingFloatChannel, default=519.6586, keys [[0,1198.2889],[599,519.6586]]}]}]
   }
   {
    type=MovieSceneFloatTrack
    display_name=Current Aperture
    sections [{type=MovieSceneFloatSection, channels [None {type=MovieSceneScriptingFloatChannel, default=0.6}]}]
   }
  ]
 }
 Cine Camera Actor {
  id=#2
  possessed_class=CineCameraActor
  tracks [
   {
    type=MovieScene3DTransformTrack
    sections [{
      type=MovieScene3DTransformSection
      channels [
       Location.X {type=MovieSceneScriptingDoubleChannel, default=-5950, keys [[0,-5950],[599,-6381.1821],[600,-6295.4043]]}
       Location.Y {type=MovieSceneScriptingDoubleChannel, default=-1008, keys [[0,-1008],[599,-622.7944],[600,-629.508]]}
       Location.Z {type=MovieSceneScriptingDoubleChannel, default=684, keys [[0,684],[599,675.6755],[600,655.3768]]}
       Rotation.X {type=MovieSceneScriptingDoubleChannel, default=-0.5254, keys [[0,-10.4533],[600,17.6072]]}
       Rotation.Y {type=MovieSceneScriptingDoubleChannel, default=-1.299, keys [[0,2.8312],[600,3.1177]]}
       Rotation.Z {type=MovieSceneScriptingDoubleChannel, default=157, keys [[0,157],[600,172.1612]]}
       Scale.X {type=MovieSceneScriptingDoubleChannel, default=1, const [1,2]}
       Scale.Y {type=MovieSceneScriptingDoubleChannel, default=1, const [1,2]}
       Scale.Z {type=MovieSceneScriptingDoubleChannel, default=1, const [1,2]}
      ]
     }]
   }
   {type=MovieSceneSpawnTrack, sections [{type=MovieSceneSpawnSection, channels [None {type=MovieSceneScriptingBoolChannel, default=true}]}]}
  ]
 }
]
display_rate {denominator=1, numerator=60}
tick_resolution {denominator=1, numerator=24000}
tracks [{type=MovieSceneCameraCutTrack, sections [{type=MovieSceneCameraCutSection, end_frame=600, start_frame=0}]}]
