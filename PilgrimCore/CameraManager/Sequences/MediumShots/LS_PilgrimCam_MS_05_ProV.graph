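# Level Sequence for a Pilgrim camera manager medium shot (MS_05, ProV variant), a single ten-second cine camera move at 60 fps.
# A spawned CineCameraActor travels and descends while rolling through level from one side to the other, with pitch and yaw held constant.
# Its camera component keeps focal length fixed while manual focus pulls closer and aperture eases to a higher f-stop.
# One camera cut section spans the whole move.
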LevelSequence /PilgrimCore/CameraManager/Sequences/MediumShots/LS_PilgrimCam_MS_05_ProV
bindings [
 CameraComponent {
  id=#1
  parent_id=#2
  possessed_class=CineCameraComponent
  tracks [
   {
    type=MovieSceneFloatTrack
    display_name=Current Focal Length
    sections [{type=MovieSceneFloatSection, channels [None {type=MovieSceneScriptingFloatChannel, default=29.2746}]}]
   }
   {
    type=MovieSceneFloatTrack
    display_name="Manual Focus Distance (Focus Settings)"
    sections [{type=MovieSceneFloatSection, channels [None {type=MovieSceneScriptingFloatChannel, default=1447.8481, keys [[0,1447.8481],[600,791.4476]]}]}]
   }
   {
    type=MovieSceneFloatTrack
    display_name=Current Aperture
    sections [{type=MovieSceneFloatSection, channels [None {type=MovieSceneScriptingFloatChannel, default=0.6, keys [[0,0.6],[599,1]]}]}]
   }
  ]
 }
 Cine Camera Actor {
  id=#2
  possessed_class=CineCameraActor
  tracks [
   {
    type=MovieScene3DTransformTrack
    sections [{
      type=MovieScene3DTransformSection
      channels [
       Location.X {type=MovieSceneScriptingDoubleChannel, default=-4621.3414, keys [[0,-6072.3335],[600,-6333.1943]]}
       Location.Y {type=MovieSceneScriptingDoubleChannel, default=-1954.016, keys [[0,241.1871],[600,74.4717]]}
       Location.Z {type=MovieSceneScriptingDoubleChannel, default=760, keys [[0,760],[600,694.7221]]}
       Rotation.X {type=MovieSceneScriptingDoubleChannel, default=-0.5254, keys [[0,11.4444],[600,-16.63]]}
       Rotation.Y {type=MovieSceneScriptingDoubleChannel, default=-1.1433, const [1.5538,2]}
       Rotation.Z {type=MovieSceneScriptingDoubleChannel, default=144.8968, const [212.5826,2]}
       Scale.X {type=MovieSceneScriptingDoubleChannel, default=1, const [1,2]}
       Scale.Y {type=MovieSceneScriptingDoubleChannel, default=1, const [1,2]}
       Scale.Z {type=MovieSceneScriptingDoubleChannel, default=1, const [1,2]}
      ]
     }]
   }
   {type=MovieSceneSpawnTrack, sections [{type=MovieSceneSpawnSection, channels [None {type=MovieSceneScriptingBoolChannel, default=true}]}]}
  ]
 }
]
display_rate {denominator=1, numerator=60}
tick_resolution {denominator=1, numerator=24000}
tracks [{type=MovieSceneCameraCutTrack, sections [{type=MovieSceneCameraCutSection, end_frame=600, start_frame=0}]}]
