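# Level Sequence for a PilgrimCam medium shot, variant 06 Portrait, that plays a single 600-frame cinematic camera move.
# A spawned CineCameraActor descends between two transform keys while tilting from looking down to slightly up and panning in yaw.
# Its camera component holds constant focal length, manual focus distance and wide aperture, with a camera shake section alongside.
# A camera cut track spans the whole move.
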LevelSequence /PilgrimCore/CameraManager/Sequences/MediumShots/LS_PilgrimCam_MS_06_Portrait
bindings [
 CameraComponent {
  id=#1
  parent_id=#2
  possessed_class=CineCameraComponent
  tracks [
   {
    type=MovieSceneFloatTrack
    display_name=Current Focal Length
    sections [{type=MovieSceneFloatSection, channels [None {type=MovieSceneScriptingFloatChannel, default=55}]}]
   }
   {
    type=MovieSceneFloatTrack
    display_name="Manual Focus Distance (Focus Settings)"
    sections [{type=MovieSceneFloatSection, channels [None {type=MovieSceneScriptingFloatChannel, default=1396.3032}]}]
   }
   {
    type=MovieSceneFloatTrack
    display_name=Current Aperture
    sections [{type=MovieSceneFloatSection, channels [None {type=MovieSceneScriptingFloatChannel, default=1}]}]
   }
   {type=MovieSceneCameraShakeTrack, sections [{type=MovieSceneCameraShakeSection, end_frame=600}]}
  ]
 }
 Cine Camera Actor {
  id=#2
  possessed_class=CineCameraActor
  tracks [
   {
    type=MovieScene3DTransformTrack
    sections [{
      type=MovieScene3DTransformSection
      channels [
       Location.X {type=MovieSceneScriptingDoubleChannel, default=-4621.3414, keys [[0,-5603.6548],[600,-5702.8859]]}
       Location.Y {type=MovieSceneScriptingDoubleChannel, default=-1954.016, keys [[0,-1165.1096],[600,-1123.3844]]}
       Location.Z {type=MovieSceneScriptingDoubleChannel, default=979, keys [[0,979],[600,632.1437]]}
       Rotation.X {type=MovieSceneScriptingDoubleChannel, default=-0.5254, keys [[0,0.1672],[600,0.1632]]}
       Rotation.Y {type=MovieSceneScriptingDoubleChannel, default=-14, keys [[0,-14],[600,5.4163]]}
       Rotation.Z {type=MovieSceneScriptingDoubleChannel, default=164, keys [[0,149],[600,164]]}
       Scale.X {type=MovieSceneScriptingDoubleChannel, default=1, const [1,2]}
       Scale.Y {type=MovieSceneScriptingDoubleChannel, default=1, const [1,2]}
       Scale.Z {type=MovieSceneScriptingDoubleChannel, default=1, const [1,2]}
      ]
     }]
   }
   {type=MovieSceneSpawnTrack, sections [{type=MovieSceneSpawnSection, channels [None {type=MovieSceneScriptingBoolChannel, default=true}]}]}
  ]
 }
]
display_rate {denominator=1, numerator=60}
tick_resolution {denominator=1, numerator=24000}
tracks [{type=MovieSceneCameraCutTrack, sections [{type=MovieSceneCameraCutSection, end_frame=600, start_frame=0}]}]
