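# Level Sequence for the PilgrimCore camera manager that plays medium shot 07 (ProV variant), a single camera move of about ten seconds at 60 fps.
# A spawned Cine Camera Actor drifts gently in position while yawing through roughly 40 degrees, with only slight roll and pitch change.
# Its camera component holds fixed focal length, manual focus distance and a very wide aperture for shallow depth of field.
# A camera cut track spans the move.
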LevelSequence /PilgrimCore/CameraManager/Sequences/MediumShots/LS_PilgrimCam_MS_07_ProV
bindings [
 CameraComponent {
  id=#1
  parent_id=#2
  possessed_class=CineCameraComponent
  tracks [
   {
    type=MovieSceneFloatTrack
    display_name=Current Focal Length
    sections [{type=MovieSceneFloatSection, channels [None {type=MovieSceneScriptingFloatChannel, default=24.3003, const [24.3003,1]}]}]
   }
   {
    type=MovieSceneFloatTrack
    display_name="Manual Focus Distance (Focus Settings)"
    sections [{type=MovieSceneFloatSection, channels [None {type=MovieSceneScriptingFloatChannel, default=307.0062, const [307.0062,1]}]}]
   }
   {
    type=MovieSceneFloatTrack
    display_name=Current Aperture
    sections [{type=MovieSceneFloatSection, channels [None {type=MovieSceneScriptingFloatChannel, default=0.6}]}]
   }
  ]
 }
 Cine Camera Actor {
  id=#2
  possessed_class=CineCameraActor
  tracks [
   {
    type=MovieScene3DTransformTrack
    sections [{
      type=MovieScene3DTransformSection
      channels [
       Location.X {type=MovieSceneScriptingDoubleChannel, default=1240.2497, keys [[2,-7238.8286],[311,-7203.394],[599,-7131.7271]]}
       Location.Y {type=MovieSceneScriptingDoubleChannel, default=-294.2261, keys [[2,-414.5265],[311,-516.7296],[599,-604.2587]]}
       Location.Z {type=MovieSceneScriptingDoubleChannel, default=722, keys [[2,722],[311,717.3204],[599,730.8279]]}
       Rotation.X {type=MovieSceneScriptingDoubleChannel, default=0, keys [[2,-0.568],[599,0.1487]]}
       Rotation.Y {type=MovieSceneScriptingDoubleChannel, default=3, keys [[2,3],[599,3.4378]]}
       Rotation.Z {type=MovieSceneScriptingDoubleChannel, default=-38.1424, keys [[2,-19.3799],[311,0.1452],[599,20.7806]]}
       Scale.X {type=MovieSceneScriptingDoubleChannel, default=1, const [1,2]}
       Scale.Y {type=MovieSceneScriptingDoubleChannel, default=1, const [1,2]}
       Scale.Z {type=MovieSceneScriptingDoubleChannel, default=1, const [1,2]}
      ]
     }]
   }
   {type=MovieSceneSpawnTrack, sections [{type=MovieSceneSpawnSection, channels [None {type=MovieSceneScriptingBoolChannel, default=true}]}]}
  ]
 }
]
display_rate {denominator=1, numerator=60}
tick_resolution {denominator=1, numerator=24000}
tracks [{type=MovieSceneCameraCutTrack, sections [{type=MovieSceneCameraCutSection, end_frame=600, start_frame=2}]}]
