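# LS_PilgrimCam_MS_09 is a Level Sequence for one medium camera shot in the PilgrimCore camera manager set, about ten seconds at 60 fps.
# A spawned CineCamera travels, pans and rolls under a single camera cut, with a camera shake section layered on.
# Its focal length widens as focus distance grows, at a fixed, very wide aperture.
# A Fort player mannequin has animation and attach tracks, but its spawn track defaults to off.
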LevelSequence /PilgrimCore/CameraManager/Sequences/MediumShots/LS_PilgrimCam_MS_09
bindings [
 Fort Player Mannequin {
  id=#1
  possessed_class=FortPlayerMannequin
  tracks [
   {
    type=MovieScene3DTransformTrack
    sections [{
      type=MovieScene3DTransformSection
      channels [
       Location.X {type=MovieSceneScriptingDoubleChannel, default=-6958}
       Location.Y {type=MovieSceneScriptingDoubleChannel, default=-587}
       Location.Z {type=MovieSceneScriptingDoubleChannel, default=475.675}
       Rotation.X {type=MovieSceneScriptingDoubleChannel, default=0}
       Rotation.Y {type=MovieSceneScriptingDoubleChannel, default=0}
       Rotation.Z {type=MovieSceneScriptingDoubleChannel, default=-90}
       Scale.X {type=MovieSceneScriptingDoubleChannel, default=1}
       Scale.Y {type=MovieSceneScriptingDoubleChannel, default=1}
       Scale.Z {type=MovieSceneScriptingDoubleChannel, default=1}
      ]
     }]
   }
   {type=MovieSceneSkeletalAnimationTrack, sections [{type=MovieSceneSkeletalAnimationSection, end_frame=709, start_frame=0}]}
   {type=MovieSceneSpawnTrack, sections [{type=MovieSceneSpawnSection, channels [None {type=MovieSceneScriptingBoolChannel, default=false}]}]}
   {type=MovieScene3DAttachTrack, sections [{type=MovieScene3DAttachSection, end_frame=709, start_frame=0}]}
  ]
 }
 CameraComponent {
  id=#2
  parent_id=#3
  possessed_class=CineCameraComponent
  tracks [
   {
    type=MovieSceneFloatTrack
    display_name=Current Focal Length
    sections [{type=MovieSceneFloatSection, channels [None {type=MovieSceneScriptingFloatChannel, default=16.3369, keys [[0,24.3296],[599,16.3369]]}]}]
   }
   {
    type=MovieSceneFloatTrack
    display_name="Manual Focus Distance (Focus Settings)"
    sections [{type=MovieSceneFloatSection, channels [None {type=MovieSceneScriptingFloatChannel, default=1950.1458, keys [[0,1728.2346],[599,1950.1458]]}]}]
   }
   {
    type=MovieSceneFloatTrack
    display_name=Current Aperture
    sections [{type=MovieSceneFloatSection, channels [None {type=MovieSceneScriptingFloatChannel, default=0.2}]}]
   }
   {type=MovieSceneCameraShakeTrack, sections []}
  ]
 }
 Cine Camera Actor {
  id=#3
  possessed_class=CineCameraActor
  tracks [
   {
    type=MovieScene3DTransformTrack
    sections [{
      type=MovieScene3DTransformSection
      channels [
       Location.X {type=MovieSceneScriptingDoubleChannel, default=-5765.3414, keys [[0,-5765.3413],[600,-5077.2349]]}
       Location.Y {type=MovieSceneScriptingDoubleChannel, default=-1846.016, keys [[0,-1846.016],[600,-1380.7179]]}
       Location.Z {type=MovieSceneScriptingDoubleChannel, default=370.2314, keys [[0,586.8422],[600,666.3878]]}
       Rotation.X {type=MovieSceneScriptingDoubleChannel, default=4.2974, keys [[0,15.0056],[600,-6.5178]]}
       Rotation.Y {type=MovieSceneScriptingDoubleChannel, default=9.3971, const [9.3971,2]}
       Rotation.Z {type=MovieSceneScriptingDoubleChannel, default=149.65, keys [[0,120.9846],[600,139.3192]]}
       Scale.X {type=MovieSceneScriptingDoubleChannel, default=1, const [1,2]}
       Scale.Y {type=MovieSceneScriptingDoubleChannel, default=1, const [1,2]}
       Scale.Z {type=MovieSceneScriptingDoubleChannel, default=1, const [1,2]}
      ]
     }]
   }
   {type=MovieSceneSpawnTrack, sections [{type=MovieSceneSpawnSection, channels [None {type=MovieSceneScriptingBoolChannel, default=true}]}]}
   {type=MovieSceneCameraShakeTrack, sections [{type=MovieSceneCameraShakeSection, end_frame=608}]}
  ]
 }
]
display_rate {denominator=1, numerator=60}
tick_resolution {denominator=1, numerator=24000}
tracks [{type=MovieSceneCameraCutTrack, sections [{type=MovieSceneCameraCutSection, end_frame=600, start_frame=0}]}]
