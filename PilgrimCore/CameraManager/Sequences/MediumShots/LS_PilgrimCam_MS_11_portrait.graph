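# Portrait-orientation medium shot 11 for the Pilgrim camera manager, a Level Sequence with a single camera cut to a spawned cine camera.
# The camera tracks sideways while easing toward the subject position with slight rotation drift, layered with a camera shake.
# Its lens uses a wide aperture for shallow depth of field, with manual focus distance shortening over the move.
# An animated, attached Fort player mannequin is included with spawning off by default, apparently as a framing stand-in.
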LevelSequence /PilgrimCore/CameraManager/Sequences/MediumShots/LS_PilgrimCam_MS_11_portrait
bindings [
 Fort Player Mannequin {
  id=#1
  possessed_class=FortPlayerMannequin
  tracks [
   {
    type=MovieScene3DTransformTrack
    sections [{
      type=MovieScene3DTransformSection
      channels [
       Location.X {type=MovieSceneScriptingDoubleChannel, default=-6914.6818}
       Location.Y {type=MovieSceneScriptingDoubleChannel, default=-462.8613}
       Location.Z {type=MovieSceneScriptingDoubleChannel, default=624.1807}
       Rotation.X {type=MovieSceneScriptingDoubleChannel, default=0}
       Rotation.Y {type=MovieSceneScriptingDoubleChannel, default=0}
       Rotation.Z {type=MovieSceneScriptingDoubleChannel, default=-90}
       Scale.X {type=MovieSceneScriptingDoubleChannel, default=1}
       Scale.Y {type=MovieSceneScriptingDoubleChannel, default=1}
       Scale.Z {type=MovieSceneScriptingDoubleChannel, default=1}
      ]
     }]
   }
   {type=MovieSceneSkeletalAnimationTrack, sections [{type=MovieSceneSkeletalAnimationSection, end_frame=709, start_frame=0}]}
   {type=MovieSceneSpawnTrack, sections [{type=MovieSceneSpawnSection, channels [None {type=MovieSceneScriptingBoolChannel, default=false}]}]}
   {type=MovieScene3DAttachTrack, sections [{type=MovieScene3DAttachSection, end_frame=709, start_frame=0}]}
  ]
 }
 CameraComponent {
  id=#2
  parent_id=#3
  possessed_class=CineCameraComponent
  tracks [
   {
    type=MovieSceneFloatTrack
    display_name=Current Focal Length
    sections [{type=MovieSceneFloatSection, channels [None {type=MovieSceneScriptingFloatChannel, default=18, const [24.4836,2]}]}]
   }
   {
    type=MovieSceneFloatTrack
    display_name="Manual Focus Distance (Focus Settings)"
    sections [{type=MovieSceneFloatSection, channels [None {type=MovieSceneScriptingFloatChannel, default=408.6225, keys [[0,477.5722],[600,408.6225]]}]}]
   }
   {
    type=MovieSceneFloatTrack
    display_name=Current Aperture
    sections [{type=MovieSceneFloatSection, channels [None {type=MovieSceneScriptingFloatChannel, default=0.7}]}]
   }
  ]
 }
 Cine Camera Actor {
  id=#3
  possessed_class=CineCameraActor
  tracks [
   {
    type=MovieScene3DTransformTrack
    sections [{
      type=MovieScene3DTransformSection
      channels [
       Location.X {type=MovieSceneScriptingDoubleChannel, default=-4621.3414, keys [[0,-6428.9062],[600,-6504.4927]]}
       Location.Y {type=MovieSceneScriptingDoubleChannel, default=-771, keys [[-25,-336],[629,-771]]}
       Location.Z {type=MovieSceneScriptingDoubleChannel, default=702.2314, keys [[0,703.4464],[600,697.7677]]}
       Rotation.X {type=MovieSceneScriptingDoubleChannel, default=4.2974, keys [[0,-0.1924],[600,-0.1923]]}
       Rotation.Y {type=MovieSceneScriptingDoubleChannel, default=9.3971, keys [[0,2.2128],[600,2.7348]]}
       Rotation.Z {type=MovieSceneScriptingDoubleChannel, default=175.5229, keys [[0,178.4866],[600,179.0746]]}
       Scale.X {type=MovieSceneScriptingDoubleChannel, default=1, const [1,2]}
       Scale.Y {type=MovieSceneScriptingDoubleChannel, default=1, const [1,2]}
       Scale.Z {type=MovieSceneScriptingDoubleChannel, default=1, const [1,2]}
      ]
     }]
   }
   {type=MovieSceneSpawnTrack, sections [{type=MovieSceneSpawnSection, channels [None {type=MovieSceneScriptingBoolChannel, default=true}]}]}
   {type=MovieSceneCameraShakeTrack, sections [{type=MovieSceneCameraShakeSection, end_frame=683}]}
  ]
 }
]
display_rate {denominator=1, numerator=60}
tick_resolution {denominator=1, numerator=24000}
tracks [{type=MovieSceneCameraCutTrack, sections [{type=MovieSceneCameraCutSection, end_frame=600, start_frame=0}]}]
