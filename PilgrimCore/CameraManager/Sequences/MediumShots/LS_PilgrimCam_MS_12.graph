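# Medium-shot camera sequence 12 for the Pilgrim camera manager: a ten-second cine camera move cut in for the whole shot.
# The camera arcs around the performer position while dipping and rising, with a fixed focal length, wide aperture, slowly increasing manual focus distance and a camera shake.
# An animated, attached Fort player mannequin marks the subject position but has its spawn track set to false, apparently serving as a framing stand-in.
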LevelSequence /PilgrimCore/CameraManager/Sequences/MediumShots/LS_PilgrimCam_MS_12
bindings [
 Fort Player Mannequin {
  id=#1
  possessed_class=FortPlayerMannequin
  tracks [
   {
    type=MovieScene3DTransformTrack
    sections [{
      type=MovieScene3DTransformSection
      channels [
       Location.X {type=MovieSceneScriptingDoubleChannel, default=-6958}
       Location.Y {type=MovieSceneScriptingDoubleChannel, default=-587}
       Location.Z {type=MovieSceneScriptingDoubleChannel, default=475.675}
       Rotation.X {type=MovieSceneScriptingDoubleChannel, default=0}
       Rotation.Y {type=MovieSceneScriptingDoubleChannel, default=0}
       Rotation.Z {type=MovieSceneScriptingDoubleChannel, default=-90}
       Scale.X {type=MovieSceneScriptingDoubleChannel, default=1}
       Scale.Y {type=MovieSceneScriptingDoubleChannel, default=1}
       Scale.Z {type=MovieSceneScriptingDoubleChannel, default=1}
      ]
     }]
   }
   {type=MovieSceneSkeletalAnimationTrack, sections [{type=MovieSceneSkeletalAnimationSection, end_frame=709, start_frame=0}]}
   {type=MovieSceneSpawnTrack, sections [{type=MovieSceneSpawnSection, channels [None {type=MovieSceneScriptingBoolChannel, default=false}]}]}
   {type=MovieScene3DAttachTrack, sections [{type=MovieScene3DAttachSection, end_frame=709, start_frame=0}]}
  ]
 }
 CameraComponent {
  id=#2
  parent_id=#3
  possessed_class=CineCameraComponent
  tracks [
   {
    type=MovieSceneFloatTrack
    display_name=Current Focal Length
    sections [{type=MovieSceneFloatSection, channels [None {type=MovieSceneScriptingFloatChannel, default=30, const [30,1]}]}]
   }
   {
    type=MovieSceneFloatTrack
    display_name="Manual Focus Distance (Focus Settings)"
    sections [{type=MovieSceneFloatSection, channels [None {type=MovieSceneScriptingFloatChannel, default=1074.2777, keys [[0,984.2781],[600,1074.2777]]}]}]
   }
   {
    type=MovieSceneFloatTrack
    display_name=Current Aperture
    sections [{type=MovieSceneFloatSection, channels [None {type=MovieSceneScriptingFloatChannel, default=0.9}]}]
   }
   {type=MovieSceneCameraShakeTrack, sections [{type=MovieSceneCameraShakeSection, end_frame=600}]}
  ]
 }
 Cine Camera Actor {
  id=#3
  possessed_class=CineCameraActor
  tracks [
   {
    type=MovieScene3DTransformTrack
    sections [{
      type=MovieScene3DTransformSection
      channels [
       Location.X {type=MovieSceneScriptingDoubleChannel, default=-4621.3414, thinned_from=5, keys [[0,-6252.6328],[600,-6169.7339]]}
       Location.Y {type=MovieSceneScriptingDoubleChannel, default=-1954.016, thinned_from=5, keys [[0,96.4467],[600,-1082.9937]]}
       Location.Z {type=MovieSceneScriptingDoubleChannel, default=702.2314, keys [[0,756.3222],[135,675.2341],[244,654.7855],[389,663.7282],[600,747.7498]]}
       Rotation.X {type=MovieSceneScriptingDoubleChannel, default=4.2974, keys [[0,0.504],[244,-0.2006],[600,-1.2274]]}
       Rotation.Y {type=MovieSceneScriptingDoubleChannel, default=9.3971, keys [[0,-0.5719],[244,3.9865],[600,-0.1208]]}
       Rotation.Z {type=MovieSceneScriptingDoubleChannel, default=124.65, thinned_from=3, keys [[0,211.1906],[600,151.455]]}
       Scale.X {type=MovieSceneScriptingDoubleChannel, default=1, const [1,2]}
       Scale.Y {type=MovieSceneScriptingDoubleChannel, default=1, const [1,2]}
       Scale.Z {type=MovieSceneScriptingDoubleChannel, default=1, const [1,2]}
      ]
     }]
   }
   {type=MovieSceneSpawnTrack, sections [{type=MovieSceneSpawnSection, channels [None {type=MovieSceneScriptingBoolChannel, default=true}]}]}
  ]
 }
]
display_rate {denominator=1, numerator=60}
tick_resolution {denominator=1, numerator=24000}
tracks [{type=MovieSceneCameraCutTrack, sections [{type=MovieSceneCameraCutSection, end_frame=600, start_frame=0}]}]
