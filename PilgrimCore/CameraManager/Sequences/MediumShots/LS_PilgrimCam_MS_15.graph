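# Level Sequence for medium shot 15 in the PilgrimCore camera manager set: a wide-lens, very wide-aperture cine camera move.
# The spawned camera actor holds a fixed transform while its CameraComponent animates relative to it, travelling and rising, then settling early in the camera cut.
# Keyed manual focus accompanies the move, and a camera shake track layers on top.
# A FortPlayerMannequin binding carries animation and attach tracks, but its spawn track defaults to off.
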LevelSequence /PilgrimCore/CameraManager/Sequences/MediumShots/LS_PilgrimCam_MS_15
bindings [
 Cine Camera Actor {
  id=#1
  possessed_class=CineCameraActor
  tracks [
   {type=MovieSceneSpawnTrack, sections [{type=MovieSceneSpawnSection, channels [None {type=MovieSceneScriptingBoolChannel, default=true}]}]}
   {
    type=MovieScene3DTransformTrack
    sections [{
      type=MovieScene3DTransformSection
      channels [
       Location.X {type=MovieSceneScriptingDoubleChannel, default=-6731.3337}
       Location.Y {type=MovieSceneScriptingDoubleChannel, default=-426.535}
       Location.Z {type=MovieSceneScriptingDoubleChannel, default=639.8296}
       Rotation.X {type=MovieSceneScriptingDoubleChannel, default=0.0915}
       Rotation.Y {type=MovieSceneScriptingDoubleChannel, default=1.5318}
       Rotation.Z {type=MovieSceneScriptingDoubleChannel, default=168.5215}
       Scale.X {type=MovieSceneScriptingDoubleChannel, default=1}
       Scale.Y {type=MovieSceneScriptingDoubleChannel, default=1}
       Scale.Z {type=MovieSceneScriptingDoubleChannel, default=1}
      ]
     }]
   }
  ]
 }
 CameraComponent {
  id=#2
  parent_id=#1
  possessed_class=CineCameraComponent
  tracks [
   {
    type=MovieSceneFloatTrack
    display_name=Current Focal Length
    sections [{type=MovieSceneFloatSection, channels [None {type=MovieSceneScriptingFloatChannel, default=20, const [20,1]}]}]
   }
   {
    type=MovieSceneFloatTrack
    display_name="Manual Focus Distance (Focus Settings)"
    sections [{
      type=MovieSceneFloatSection
      channels [None {type=MovieSceneScriptingFloatChannel, default=517.0811, keys [[0,517.0811],[84,285.5822],[122,351.0704],[236,283.6609]]}]
     }]
   }
   {
    type=MovieSceneFloatTrack
    display_name=Current Aperture
    sections [{type=MovieSceneFloatSection, channels [None {type=MovieSceneScriptingFloatChannel, default=0.6}]}]
   }
   {type=MovieSceneCameraShakeTrack, sections [{type=MovieSceneCameraShakeSection, end_frame=622}]}
   {
    type=MovieScene3DTransformTrack
    sections [{
      type=MovieScene3DTransformSection
      channels [
       Location.X {type=MovieSceneScriptingDoubleChannel, default=1051.4999, keys [[0,-433.8758],[40,-360.1301],[84,-206.135],[122,-92.524],[236,-43.2235]]}
       Location.Y {type=MovieSceneScriptingDoubleChannel, default=199, keys [[0,199],[40,224],[84,261.7301],[122,251.3148],[236,246.7946]]}
       Location.Z {type=MovieSceneScriptingDoubleChannel, default=41, keys [[0,-74],[40,-12],[84,2],[122,36],[236,41]]}
       Rotation.X {type=MovieSceneScriptingDoubleChannel, default=-0.4975, keys [[0,-0.4975],[35,-6.1051],[84,4.2483],[122,-0.4975],[236,-0.4975]]}
       Rotation.Y {type=MovieSceneScriptingDoubleChannel, default=6, keys [[0,16.9878],[37,6],[82,9],[122,5],[236,5]]}
       Rotation.Z {type=MovieSceneScriptingDoubleChannel, default=-27.008, keys [[0,-27.008],[39,-28.4091],[86,-24.9943],[122,-27.008],[209,-22.5659]]}
       Scale.X {type=MovieSceneScriptingDoubleChannel, default=1, const [1,5]}
       Scale.Y {type=MovieSceneScriptingDoubleChannel, default=1, const [1,5]}
       Scale.Z {type=MovieSceneScriptingDoubleChannel, default=1, const [1,5]}
      ]
     }]
   }
  ]
 }
 Fort Player Mannequin {
  id=#3
  possessed_class=FortPlayerMannequin
  tracks [
   {
    type=MovieScene3DTransformTrack
    sections [{
      type=MovieScene3DTransformSection
      channels [
       Location.X {type=MovieSceneScriptingDoubleChannel, default=-6914.6818}
       Location.Y {type=MovieSceneScriptingDoubleChannel, default=-462.8613}
       Location.Z {type=MovieSceneScriptingDoubleChannel, default=625.9081}
       Rotation.X {type=MovieSceneScriptingDoubleChannel, default=0}
       Rotation.Y {type=MovieSceneScriptingDoubleChannel, default=0}
       Rotation.Z {type=MovieSceneScriptingDoubleChannel, default=-90}
       Scale.X {type=MovieSceneScriptingDoubleChannel, default=1}
       Scale.Y {type=MovieSceneScriptingDoubleChannel, default=1}
       Scale.Z {type=MovieSceneScriptingDoubleChannel, default=1}
      ]
     }]
   }
   {type=MovieSceneSkeletalAnimationTrack, sections [{type=MovieSceneSkeletalAnimationSection, end_frame=709, start_frame=0}]}
   {type=MovieScene3DAttachTrack, sections [{type=MovieScene3DAttachSection, end_frame=709, start_frame=0}]}
   {type=MovieSceneSpawnTrack, sections [{type=MovieSceneSpawnSection, channels [None {type=MovieSceneScriptingBoolChannel, default=false}]}]}
  ]
 }
]
display_rate {denominator=1, numerator=60}
tick_resolution {denominator=1, numerator=24000}
tracks [{type=MovieSceneCameraCutTrack, sections [{type=MovieSceneCameraCutSection, end_frame=600, start_frame=0}]}]
