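# Level Sequence for a PilgrimCore camera manager portrait medium shot, running about ten seconds at 60 fps on one camera cut.
# A spawned CineCamera's component moves along local X while rolling and panning, as its focal length widens.
# Manual focus distance is keyed over the shot, aperture stays wide at a fixed value, and the camera shake track is empty.
# A player mannequin binding carries animation and attach tracks but has spawning off by default.
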LevelSequence /PilgrimCore/CameraManager/Sequences/MediumShots/LS_PilgrimCam_MS_16_Portrait
bindings [
 Fort Player Mannequin {
  id=#1
  possessed_class=FortPlayerMannequin
  tracks [
   {
    type=MovieScene3DTransformTrack
    sections [{
      type=MovieScene3DTransformSection
      channels [
       Location.X {type=MovieSceneScriptingDoubleChannel, default=-6914.6818}
       Location.Y {type=MovieSceneScriptingDoubleChannel, default=-462.8613}
       Location.Z {type=MovieSceneScriptingDoubleChannel, default=625.6368}
       Rotation.X {type=MovieSceneScriptingDoubleChannel, default=0}
       Rotation.Y {type=MovieSceneScriptingDoubleChannel, default=0}
       Rotation.Z {type=MovieSceneScriptingDoubleChannel, default=-90}
       Scale.X {type=MovieSceneScriptingDoubleChannel, default=1}
       Scale.Y {type=MovieSceneScriptingDoubleChannel, default=1}
       Scale.Z {type=MovieSceneScriptingDoubleChannel, default=1}
      ]
     }]
   }
   {type=MovieSceneSkeletalAnimationTrack, sections [{type=MovieSceneSkeletalAnimationSection, end_frame=709, start_frame=0}]}
   {type=MovieSceneSpawnTrack, sections [{type=MovieSceneSpawnSection, channels [None {type=MovieSceneScriptingBoolChannel, default=false}]}]}
   {type=MovieScene3DAttachTrack, sections [{type=MovieScene3DAttachSection, end_frame=709, start_frame=0}]}
  ]
 }
 CameraComponent {
  id=#2
  parent_id=#3
  possessed_class=CineCameraComponent
  tracks [
   {
    type=MovieSceneFloatTrack
    display_name=Current Focal Length
    sections [{type=MovieSceneFloatSection, channels [None {type=MovieSceneScriptingFloatChannel, default=32.0741, keys [[0,33.8893],[600,25.0103]]}]}]
   }
   {
    type=MovieSceneFloatTrack
    display_name="Manual Focus Distance (Focus Settings)"
    sections [{
      type=MovieSceneFloatSection
      channels [None {type=MovieSceneScriptingFloatChannel, default=407.8211, keys [[0,365.234],[599,407.8211],[600,337.5237]]}]
     }]
   }
   {
    type=MovieSceneFloatTrack
    display_name=Current Aperture
    sections [{type=MovieSceneFloatSection, channels [None {type=MovieSceneScriptingFloatChannel, default=1}]}]
   }
   {type=MovieSceneCameraShakeTrack, sections []}
   {
    type=MovieScene3DTransformTrack
    sections [{
      type=MovieScene3DTransformSection
      channels [
       Location.X {type=MovieSceneScriptingDoubleChannel, default=-293, keys [[0,-293],[600,0]]}
       Location.Y {type=MovieSceneScriptingDoubleChannel, default=9.3388, const [0,2]}
       Location.Z {type=MovieSceneScriptingDoubleChannel, default=-171.0472, const [0,2]}
       Rotation.X {type=MovieSceneScriptingDoubleChannel, default=-0.4975, keys [[0,0],[600,28.6642]]}
       Rotation.Y {type=MovieSceneScriptingDoubleChannel, default=-24}
       Rotation.Z {type=MovieSceneScriptingDoubleChannel, default=8, keys [[-22,-6],[622,8]]}
       Scale.X {type=MovieSceneScriptingDoubleChannel, default=1, const [1,2]}
       Scale.Y {type=MovieSceneScriptingDoubleChannel, default=1, const [1,2]}
       Scale.Z {type=MovieSceneScriptingDoubleChannel, default=1, const [1,2]}
      ]
     }]
   }
  ]
 }
 CineCameraActor1 {
  id=#3
  possessed_class=CineCameraActor
  tracks [{type=MovieSceneSpawnTrack, sections [{type=MovieSceneSpawnSection, channels [None {type=MovieSceneScriptingBoolChannel, default=true}]}]}]
 }
]
display_rate {denominator=1, numerator=60}
tick_resolution {denominator=1, numerator=24000}
tracks [{type=MovieSceneCameraCutTrack, sections [{type=MovieSceneCameraCutSection, end_frame=600, start_frame=0}]}]
